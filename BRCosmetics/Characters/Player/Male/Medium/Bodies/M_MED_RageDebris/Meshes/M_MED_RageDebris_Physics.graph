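# Physics asset for the RageDebris medium male player body cosmetic, set up for secondary motion on a dangling strap.
# Three capsule strap bodies chain from a small dyn_strap_MSTR root through angle-limited constraints with very little twist.
# The root joint is tight, while the two lower joints swing mostly along one axis.
# Unconstrained pelvis, lower spine and thigh capsules appear to serve as colliders for the strap.
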PhysicsAsset /BRCosmetics/Characters/Player/Male/Medium/Bodies/M_MED_RageDebris/Meshes/M_MED_RageDebris_Physics
bodies {
 dyn_strap_01 [{shapeType=Capsule, length=3.9659, radius=3.2366, shapeName=dyn_strap_01_capsule, center=<5.8263 0.2637 2.4>, rotation=<rot 90 0 0>}]
 dyn_strap_02 [{shapeType=Capsule, length=3.2003, radius=3.3543, shapeName=dyn_strap_02_capsule, center=<4.7292 0.2637 2.4>, rotation=<rot 90 0 0>}]
 dyn_strap_03 [{shapeType=Capsule, length=1.5242, radius=4.2581, shapeName=dyn_strap_03_capsule, center=<2.5542 -0.0587 2.4>, rotation=<rot 0 0 90>}]
 dyn_strap_MSTR [{shapeType=Capsule, length=0.505, radius=0.505, shapeName=dyn_strap_MSTR_capsule, rotation=<rot 0 -90 0>}]
 pelvis [{shapeType=Capsule, length=14.201, radius=10.2266, shapeName=pelvis_capsule, center=<-10.6557 0 0>}]
 spine_01 [{shapeType=Capsule, length=14.0111, radius=9.8735, shapeName=spine_01_capsule, center=<-12.2434 0.332 0>, rotation=<rot 0 -90 0>}]
 thigh_l [{shapeType=Capsule, length=35.1522, radius=9.7, shapeName=thigh_l_capsule, center=<-16.9205 0 0>, rotation=<rot 90 0 0>}]
 thigh_r [{shapeType=Capsule, length=35.8169, radius=9.7, shapeName=thigh_r_capsule, center=<17.6272 0 0>, rotation=<rot -90 180 180>}]
}
constraints [
 {bone1Name=dyn_strap_02, bone2Name=dyn_strap_01, swing1LimitDegrees=5, swing2LimitDegrees=30, twistLimitDegrees=2}
 {bone1Name=dyn_strap_03, bone2Name=dyn_strap_02, swing1LimitDegrees=5, swing2LimitDegrees=24, twistLimitDegrees=2}
 {bone1Name=dyn_strap_01, bone2Name=dyn_strap_MSTR, swing1LimitDegrees=5, swing2LimitDegrees=5, twistLimitDegrees=2}
]
